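# Translucent unlit Niagara sprite material for vehicle cosmetic smoke puffs with glowing glitter.
# Glitter multiplies one noise channel sampled at two tilings, then boosts, tints and adds it to the particle-colored smoke flipbook texture through an emissive adjustment function.
# Opacity adds glitter, scaled by dynamic GlitterGlow, to particle-alpha smoke shaped by dynamic NoiseTransition, then passes through depth fade.
# Dynamic GlitterGlow affects only glitter opacity, not its emissive.
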Material /VehicleCosmetics/FX/Materials/Smoke/M_VCVFX_Smoke_Glitter
graph {
 properties {
  bUsedWithNiagaraSprites=true
  blendMode=BLEND_Translucent
  pixelDepthOffsetMode=PDOM_Legacy
  refractionDepthBias=1
  shadingModel=MSM_Unlit
  usedShadingModels=Unlit
 }
 nodes [
  n1 = VectorParameter() {parameterName=GlitterColor, defaultValue=<rgba 1 0.4176 0 1>}
  n2 = TextureCoordinate() {uTiling=10, vTiling=10}
  n3 = Constant() {r=0.5}
  n4 = Multiply(
   A=TextureSampleParameter2D(UVs=@n2) {parameterName=GlitterTexture, texture=/DealershipSharedLibrary/FX/Textures/Noise/T_VCVFX_NoiseGeneric01Pack_RGB}.B
   B=TextureSampleParameter2D(UVs=Multiply(A=@n2, B=@n3)) {parameterName=GlitterTexture, texture=&T_VCVFX_NoiseGeneric01Pack_RGB}.B
  )
  n5 = Saturate(None=Multiply(A=Desaturation(None=@n4)) {constB=10})
  n6 = Multiply(A=Multiply(A=Multiply(A=@n1.RGB, B=@n5), B=Constant() {r=1024}), B=ScalarParameter() {defaultValue=1, parameterName=GlitterGlow})
  n7 = DynamicParameter() {paramNames [NoiseTransition, GlitterGlow, Distortion, Param4]}
  n8 = Desaturation(None=Multiply(A=@n6, B=@n7.GlitterGlow))
  n9 = ParticleColor()
  n10 = TextureSampleParameter2D() {parameterName=MainTexture, texture=/DealershipSharedLibrary/FX/Textures/Flipbook/T_VCVFX_smoke_puff_8x8}
  n11 = LinearInterpolate(A=@n10.A, B=@n10.A, Alpha=ScalarParameter() {parameterName=Alpha/RGB})
  n12 = Multiply(A=Saturate(None=Multiply(A=@n7.NoiseTransition, B=@n11)), B=@n11)
  MaterialFunctionCall(
   Opacity In=Add(A=Saturate(None=@n8), B=Multiply(A=@n9.A, B=@n12))
   DepthFade Distance=ScalarParameter() {defaultValue=50, group=Depth Fade, parameterName=DepthFade Distance}
  ) {materialFunction=&MF_VCVFX_DepthFade}
  n13 = Multiply(A=@n10.B, B=ScalarParameter() {defaultValue=1, group=Main Texture, parameterName=Texture Multiplier})
  MaterialFunctionCall(Emissive=Add(A=@n6, B=Multiply(A=@n13, B=@n9.RGB))) {materialFunction=&MF_FX_EmissiveAdjustments}
 ]
}
